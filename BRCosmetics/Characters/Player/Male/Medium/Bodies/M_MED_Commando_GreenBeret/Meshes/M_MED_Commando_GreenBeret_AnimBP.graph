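# Animation Blueprint for the male medium Commando Green Beret body mesh, adding secondary motion to its left pouch and armband.
# On initialization it stores the result of GetOwnerSkeletalMeshComponent as MeshToCopy, whose pose Copy Pose From Mesh brings in.
# That pose runs through chained Anim Dynamics nodes: a low-gravity pouch with locked rotation and slight linear travel, then three angular-sprung armband bones simulated in actor space.
# The pouch node carries an LOD threshold.
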AnimBlueprint /BRCosmetics/Characters/Player/Male/Medium/Bodies/M_MED_Commando_GreenBeret/Meshes/M_MED_Commando_GreenBeret_AnimBP
target_skeleton=&/M_MED_Commando_GreenBeret_Skeleton
graphs [
 AnimGraph {nodes [
   n1 = Get MeshToCopy() -> (MeshToCopy:Skeletal Mesh Component Object Reference)
   n2 = Anim Dynamics - Bone: dyn_pouch_l(ComponentPose=Local To Component(LocalPose=Copy Pose From Mesh(SourceMeshComponent=@n1)), Alpha=1, ExternalForce="0,0,0") {
    bGravityOverrideInSimSpace=true
    gravityScale=0.1
    lODThreshold=1
    alphaBoolBlend {blendOption=Linear}
    alphaScaleBiasClamp {bClampResult=true, clampMax=0.33, clampMin=0.1, scale=0.0033}
    boundBone {boneName=dyn_pouch_l}
    physicsBodyDefinitions [{
      sphereCollisionRadius=0
      boundBone {boneName=dyn_pouch_l}
      boxExtents=<5 5 5>
      constraintSetup {angularTargetAxis=AxisY, twistAxis=AxisY, angularLimitsMax=<0 0 0>, angularLimitsMin=<0 0 0>, angularTarget=<0 0 0>, linearAxesMax=<0 0 2>}
      localJointOffset=<0 0 5>
     }]
   }
   n3 = Anim Dynamics - Bone: dyn_armband_knot(ComponentPose=@n2, Alpha=0.75, ExternalForce="0,0,0") {
    alpha=0.9
    angularSpringConstant=1500
    bAngularSpring=true
    bGravityOverrideInSimSpace=true
    numSolverIterationsPostUpdate=4
    numSolverIterationsPreUpdate=16
    simulationSpace=Actor
    alphaBoolBlend {blendOption=Linear}
    boundBone {boneName=dyn_armband_knot}
    physicsBodyDefinitions [{
      sphereCollisionRadius=0
      boundBone {boneName=dyn_armband_knot}
      boxExtents=<15 5 5>
      constraintSetup {angularTargetAxis=AxisZ, twistAxis=AxisY, angularLimitsMax=<15 0 15>, angularLimitsMin=<-15 0 -15>, angularTarget=<0 0 1>}
      localJointOffset=<0 0 0>
     }]
   }
   n4 = Anim Dynamics - Bone: dyn_armband_low(ComponentPose=@n3, Alpha=0.5, ExternalForce="0,0,0") {
    alpha=0.9
    angularSpringConstant=1500
    bAngularSpring=true
    bGravityOverrideInSimSpace=true
    numSolverIterationsPostUpdate=4
    numSolverIterationsPreUpdate=16
    simulationSpace=Actor
    alphaBoolBlend {blendOption=Linear}
    boundBone {boneName=dyn_armband_low}
    physicsBodyDefinitions [{
      sphereCollisionRadius=0
      boundBone {boneName=dyn_armband_low}
      boxExtents=<15 5 5>
      constraintSetup {angularTargetAxis=AxisZ, twistAxis=AxisY, angularLimitsMax=<15 15 15>, angularLimitsMin=<-15 -15 -15>, angularTarget=<0 0 1>}
      localJointOffset=<0 0 0>
     }]
   }
   n5 = Anim Dynamics - Bone: dyn_armband_hi(ComponentPose=@n4, Alpha=0.5, ExternalForce="0,0,0") {
    alpha=0.9
    angularSpringConstant=1500
    bAngularSpring=true
    bGravityOverrideInSimSpace=true
    numSolverIterationsPostUpdate=4
    numSolverIterationsPreUpdate=16
    simulationSpace=Actor
    alphaBoolBlend {blendOption=Linear}
    boundBone {boneName=dyn_armband_hi}
    physicsBodyDefinitions [{
      sphereCollisionRadius=0
      boundBone {boneName=dyn_armband_hi}
      boxExtents=<15 5 5>
      constraintSetup {angularTargetAxis=AxisZ, twistAxis=AxisY, angularLimitsMax=<15 15 15>, angularLimitsMin=<-15 -15 -15>, angularTarget=<0 0 1>}
      localJointOffset=<0 0 0>
     }]
   }
   Output Pose(Result=Component To Local(ComponentPose=@n5)) {layerGroup=None}
  ]}
 EventGraph {nodes [
   Event BlueprintInitializeAnimation()
   > Set MeshToCopy(MeshToCopy:Skeletal Mesh Component Object Reference=GetOwnerSkeletalMeshComponent()) -> (Output_Get:Skeletal Mesh Component Object Reference)
   Event BlueprintUpdateAnimation()
  ]}
]
variables [MeshToCopy {type=SkeletalMeshComponent}]
